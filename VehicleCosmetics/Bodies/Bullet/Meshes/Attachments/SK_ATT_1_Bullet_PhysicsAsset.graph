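# Physics asset for SK_ATT_1_Bullet, an attachment skeletal mesh for the Bullet body in vehicle cosmetics.
# It holds a single body, named body, made of one very small, rotated capsule offset far along X and slightly above the origin.
# It has no constraints, so it defines no joints or ragdoll chain.
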PhysicsAsset /VehicleCosmetics/Bodies/Bullet/Meshes/Attachments/SK_ATT_1_Bullet_PhysicsAsset
bodies {body [{shapeType=Capsule, length=0.505, radius=0.505, center=<283.4192 0 13.8653>, rotation=<rot 0 -90 -65.5948>}]}
constraints []
