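# Physics asset for the SK_ATT_2_Bullet skeletal mesh, an attachment of the Bullet vehicle body in the VehicleCosmetics content.
# It has a single body named "body", made of one large capsule offset forward along X and rotated away from its default upright axis.
# It defines no constraints, so there is no articulation between bodies.
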PhysicsAsset /VehicleCosmetics/Bodies/Bullet/Meshes/Attachments/SK_ATT_2_Bullet_PhysicsAsset
bodies {body [{shapeType=Capsule, length=185.8638, radius=109.922, center=<99.313 0 1.4221>, rotation=<rot 0 90 -87.8603>}]}
constraints []
