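# Physics asset for SK_ATT_3_Bullet, a skeletal mesh attachment under the Bullet vehicle body cosmetics.
# It holds a single body shaped as one large capsule, offset mostly forward along X and rotated off the default axis.
# With no constraints, it gives the attachment one rough collision volume rather than an articulated, jointed setup.
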PhysicsAsset /VehicleCosmetics/Bodies/Bullet/Meshes/Attachments/SK_ATT_3_Bullet_PhysicsAsset
bodies {body [{shapeType=Capsule, length=178.7811, radius=108.0839, center=<118.6595 0 -6.1901>, rotation=<rot 0 90 -84.3629>}]}
constraints []
