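# Physics asset for the skeletal mesh attachment SK_ATT_4_Bullet, a cosmetic part for the Bullet vehicle body.
# It holds a single body made of one long, thin capsule, offset from the origin along X, slightly lowered and rolled onto its side to lie crosswise.
# It defines no constraints, so there are no joints to other bodies.
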PhysicsAsset /VehicleCosmetics/Bodies/Bullet/Meshes/Attachments/SK_ATT_4_Bullet_PhysicsAsset
bodies {body [{shapeType=Capsule, length=104.5603, radius=8.2004, center=<231.0072 0 -15.8705>, rotation=<rot 0 0 -90>}]}
constraints []
